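# Additive unlit Niagara mesh-particle material for the HandsUp emote's "FlowStarMesh", drawing glowing flare stars that pulse with Jerky event music.
# World position offset builds sprites, scaled by camera distance and screen-size terms and drifted by panned swirl-noise textures.
# Kick, snare and hats values from the Jerky parameter collection, split by vertex color, drive size and brightness.
# Vertex color also blends flare textures and two colors, while a UV sphere mask handles reveal.
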Material /BRCosmetics/Athena/Effects/Particles/Emotes/NS_Emote_HandsUp/M_Jerky_FlowStarMesh
graph {
 properties {
  bUsedWithNiagaraMeshParticles=true
  blendMode=BLEND_Additive
  pixelDepthOffsetMode=PDOM_Legacy
  shadingModel=MSM_Unlit
  twoSided=true
  usedShadingModels=Unlit
 }
 nodes [
  n1 = CollectionParameter() {collection=/Game/Athena/Events/Jerky/MPC_JerkyCollection, parameterName=TrackStemKick}
  n2 = CollectionParameter() {collection=&MPC_JerkyCollection, parameterName=TrackStemSnare}
  n3 = CollectionParameter() {collection=&MPC_JerkyCollection, parameterName=TrackStemHats}
  n4 = Multiply(
   A=Power(Base=MaterialFunctionCall(R=@n1, G=@n2, B=@n3) {materialFunction=&AppendMany}.RGB) {constExponent=3}
   B=ScalarParameter() {defaultValue=3, parameterName=BeatSizeMult}
  )
  Add(A=@n4) {constB=0.8}
  n5 = WorldPosition()
  n6 = Constant() {r=1}
  n7 = ScalarParameter() {defaultValue=0.9, parameterName=RevealHardness}
  n8 = SphereMask(
   A=ComponentMask(None=TextureCoordinate()) {r=true}
   B=LinearInterpolate(B=Subtract(A=@n6, B=@n7), Alpha=ScalarParameter() {parameterName=Reveal}) {constA=-1, constB=0.5}
   Radius=Constant() {r=1}
   Hardness=@n7
  )
  n9 = Ceil(None=@n8)
  n10 = OneMinus(None=@n8)
  n11 = Multiply(A=Power(Base=Multiply(A=@n9, B=@n10)), B=ScalarParameter() {parameterName=SideReveal})
  n12 = Add(A=Multiply(A=@n11, B=ScalarParameter() {defaultValue=1, parameterName=SideRevealSizeAdd}))
  n13 = VertexColor()
  n14 = LinearInterpolate(A=ScalarParameter() {defaultValue=0.15, parameterName=MinSizeMult}, Alpha=@n13.R)
  n15 = VertexColor()
  n16 = Round(None=@n15.R)
  n17 = Round(None=@n15.G)
  n18 = Max(A=@n16, B=@n17)
  n19 = Multiply(A=@n16, B=@n18)
  n20 = Multiply(A=@n18, B=@n17)
  n21 = Add(A=@n19, B=@n20)
  n22 = Ceil(None=OneMinus(None=Saturate(None=@n21)))
  n23 = ScalarParameter() {defaultValue=1, parameterName=Fade}
  n24 = MaterialFunctionCall(f3=Multiply(
    A=MaterialFunctionCall(R=@n19, G=@n20, B=@n22) {materialFunction=&AppendMany}.RGB
    B=LinearInterpolate(B=@n4, Alpha=@n23) {constA=0.5}
   ) {constA=1}) {materialFunction=&AddComponents}
  n25 = Multiply(A=ScalarParameter() {defaultValue=6000, parameterName=SpriteSquareSize}, B=Add(A=Multiply(A=@n12, B=@n14), B=@n24.f3))
  n26 = CameraPositionWS()
  n27 = WorldPosition() {worldPositionShaderOffset=WPT_ExcludeAllShaderOffsets}
  n28 = Saturate(None=Divide(A=Distance(A=@n26, B=@n27.XYZ), B=ScalarParameter() {defaultValue=400000, parameterName=WPO_NearCameraFade}))
  n29 = PixelDepth()
  n30 = Multiply(A=ComponentMask(None=ViewSize()) {g=true}, B=ScalarParameter() {defaultValue=0.2, parameterName=ScreenWidth})
  n31 = Multiply(A=Add(A=Multiply(A=@n25, B=@n28), B=Divide(A=@n29.R, B=@n30)), B=ScalarParameter() {defaultValue=1, parameterName=Fade})
  n32 = Multiply(A=Saturate(None=VertexNormalWS()), B=ScalarParameter() {defaultValue=10000, parameterName="SpriteFlowIntensity(Local)"})
  n33 = MaterialFunctionCall(Vector or UVs=Subtract(A=TextureCoordinate() {coordinateIndex=1}, B=Constant() {r=0.5})) {materialFunction=&VectorToRadialValue}
  n34 = VectorParameter() {parameterName="SpriteFlowTextureCtrl(Local)", defaultValue=<rgba 3 2 0.02 0.02>}
  n35 = Multiply(A=@n33.Radial Coordinates, B=AppendVector(A=@n34.R, B=@n34.R))
  n36 = AppendVector(A=@n34.B, B=@n34.B)
  n37 = LinearInterpolate(
   A=ScalarParameter() {defaultValue=-1, parameterName=Radian_Inside}
   Alpha=TextureSample(UVs=Panner(Coordinate=@n35, Speed=@n36)) {samplerType=SAMPLERTYPE_Normal, texture=/Game/Global/Textures/Default/Noise/T_SwirlNoise_N}.R
  ) {constA=-0.5}
  n38 = VectorParameter() {parameterName="SpriteFlowTextureCtrl(Word)", defaultValue=<rgba 2 2 -0.02 -0.02>}
  n39 = Panner(Coordinate=Multiply(A=@n33.Radial Coordinates, B=AppendVector(A=@n38.R, B=@n38.R)) {constA=1}, Speed=AppendVector(A=@n38.B, B=@n38.B))
  Add(
   A=Add(
    A=MaterialFunctionCall!(Center Location=@n5.XYZ, 0-1 UVs, XY scale=@n31, OverridePivot, "Normals (Optional)") {materialFunction=&Sprite}.World position offset
    B=Multiply(A=@n32, B=@n37)
   ) {constB=0}
   B=Multiply(
    A=ScalarParameter() {defaultValue=10000, parameterName="SpriteFlowIntensity(Word)"}
    B=TextureSample(UVs=@n39) {samplerType=SAMPLERTYPE_Normal, texture=&T_SwirlNoise_N}.RGB
   )
  ) {constB=0}
  Constant()
  n40 = TextureCoordinate()
  n41 = Constant() {r=0.5}
  n42 = ScalarParameter() {defaultValue=0.16, parameterName=CenterRadius}
  n43 = Multiply(
   A=Power(Base=SphereMask(A=@n40, B=@n41, Radius=@n42) {hardnessPercent=0}, Exp=ScalarParameter() {defaultValue=25, parameterName=CenterPower}) {constExponent=5}
   B=ScalarParameter() {defaultValue=15, parameterName=CenterBrightness}
  )
  n44 = Saturate(None=TextureCoordinate())
  n45 = TextureSample(UVs=@n44) {texture=/Game/Effects/Fort_Effects/Textures/Flares/T_Flare_Spikeball}
  n46 = TextureSample(UVs=@n44) {samplerType=SAMPLERTYPE_Grayscale, texture=/Game/Effects/Fort_Effects/Textures/Flares/T_Cross_Flare}
  n47 = Constant() {r=1}
  n48 = Multiply(
   A=Multiply(A=LinearInterpolate(A=@n45.RGB, B=@n46.RGB, Alpha=@n13.G), B=Add(A=@n47, B=@n24.f3))
   B=LinearInterpolate(
    A=VectorParameter() {parameterName=Color1, defaultValue=<rgba 0.1989 0.1334 0.5156 0>}.RGB
    B=VectorParameter() {parameterName=Color2, defaultValue=<rgba 0.5198 0.8052 1 0>}.RGB
    Alpha=@n13.B
   )
  )
  MaterialFunctionCall(Emissive=Multiply(A=Add(A=@n11), B=Add(A=@n43, B=@n48))) {materialFunction=&MF_FX_EmissiveAdjustments}
  n49 = ParticleColor()
  n50 = Add(A=@n49.A) {constB=-1}
  n51 = VertexColor()
  n52 = Multiply(A=@n8) {constB=20}
  Multiply(
   A=Saturate(None=LinearInterpolate(A=@n50, B=@n49.A, Alpha=@n51.R))
   B=LinearInterpolate(B=Saturate(None=@n52), Alpha=ScalarParameter() {parameterName=SideReveal}) {constA=1}
  )
 ]
}
