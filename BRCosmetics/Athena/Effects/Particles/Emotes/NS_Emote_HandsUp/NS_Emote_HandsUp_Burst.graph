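# Niagara burst effect for the HandsUp emote that throws out a one-shot spray of short-lived star sprites.
# Two emitters, green-tinted and multicolor, share forward-biased random velocity, partial inherited velocity, gravity, drag and curl noise.
# Green stars shrink and write a curve-tinted MultColor variable, while multicolor stars randomize hue, pop in, then shrink and fade.
# Skeletal-mesh bone placement is present in the green emitter but disabled.
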NiagaraSystem /BRCosmetics/Athena/Effects/Particles/Emotes/NS_Emote_HandsUp/NS_Emote_HandsUp_Burst
dependencies {
 usedDataInterfaces [&NiagaraDataInterfaceCurve, &NiagaraDataInterfaceColorCurve]
 usedDynamicInputs []
 usedModules [/Niagara/Modules/System/SystemLifeCycle, &SystemState]
}
emitters {
 FN_GreenStars {
  emitterSpawnScript [&FN_PreserveSkydiveUserVariables() {module=PreserveUserVariables}]
  emitterUpdateScript [
   &EmitterState()
   &SpawnBurst_Instantaneous(Spawn Count=20, Use Spawn Probability=false)
  ]
  particleSpawnScript [
   /Game/Effects/Niagara/Modules/Groups/FN_SetSpawnGroup_v1/FN_SetSpawnGroup(Emitter Bone Count=2) {enabled=false}
   /Niagara/Modules/Spawn/Initialization/InitializeParticle(
    Color=&UniformRangedLinearColor(Minimum=<rgba 1 0.5939 0.26 1>, Maximum=<rgba 0.5548 0.6742 0.917 1>)
    Lifetime=&UniformRangedFloat(Minimum=0.25, Maximum=1)
    Mesh Scale=&Multiply_Vector(
     A=<1 1 1>
     B=&VectorFromFloat(Value=&UniformRangedFloat(Minimum=0.2, Maximum=1))
    )
    Module.Write Color=true
    Module.Write Lifetime=true
    Module.Write Mass=true
    Module.Write Material Random=true
    Module.Write Position=false
    Module.Write Scale=true
    Module.Write SpriteRotation=false
    Module.Write SpriteSize=false
    Position=&SimulationPosition()
   )
   &SampleSkeletalMeshSkeleton(Mesh:NiagaraDataInterfaceSkeletalMesh=$User.SkeletalMesh, Specific Bone Index:NiagaraInt32=$Particles.SpawnGroup) {enabled=false}
   /Niagara/Modules/Spawn/Location/SkeletalMeshSkeletonLocation(Module.Inherit Velocity=false, Module.Use Sampled Bone as Ribbon ID=true) {enabled=false}
   /Niagara/Modules/Spawn/Orientation/V2/InitialMeshOrientation(
    Orientation Vector=&UniformRangedVector(Minimum=<-1 -1 -1>, Maximum=<1 1 1>)
    Use Orientation Vector=true
    Use Rotation Vector=false
   )
   /SetVariables(Particles.RotationOrientation:NiagaraFloat=&UniformRangedFloat(Minimum=-1, Maximum=1)) {isSetParameters=true}
   /SetVariables(Particles.RotationSpeed:Vector3f=&UniformRangedVector(Minimum=<-1 -1 -1>, Maximum=<1 1 1>)) {isSetParameters=true}
   &InheritVelocity(Module.Limit Speed=false, Velocity Scale=<0.25 0.25 0.25>)
   &AddVelocity(Velocity=&UniformRangedVector(Minimum=<100 -500 -500>, Maximum=<1000 500 500>))
  ]
  particleUpdateScript [
   &ParticleState()
   &Drag(Drag=7, Use Linear Drag=true, Use Rotational Drag=false)
   /Niagara/Modules/Update/Forces/CurlNoiseForce(
    Noise Frequency:NiagaraFloat=&UniformRangedFloat(Minimum=0.1, Maximum=1)
    Noise Strength=&UniformRangedFloat(Minimum=300, Maximum=600)
    Pan Noise=false
    Pan Noise Field=<1 1 0>
   )
   &GravityForce(Gravity=<0 0 -500>)
   &SolveForcesAndVelocity(Clamp Velocity=false, Limit Acceleration=false)
   &ScaleSpriteSize(Scale Factor=&Multiply_Vector2(
     A=<1 1>
     B=&Vector2DFromFloat(Value=&FloatFromCurve(FloatCurve=DataInterface() {propertyValues {Curve {keys [<at -0.0046 3>,<at 1.0011 0.002>]}}}))
    ))
   /SetVariables(Particles.MultColor:LinearColor=&MakeLinearColorFromVectorAndFloat(
     "Vector (RGB)"=&Multiply_Vector(
      A=&MakeVectorFromLinearColorRGB(Color=$Particles.Color)
      B=&MakeVectorFromLinearColorRGB(Color=&ColorFromCurve(ColorCurve=DataInterface() {propertyValues {
          AlphaCurve {keys [<at -0.0023 1 RCIM_Cubic>,<at 0.9977 1 RCIM_Cubic>]}
          BlueCurve {keys [<at 0.1285 0.117 RCIM_Cubic>,<at 0.3093 0.086>,<at 0.5661 0.7784>,<at 0.7637 0.057>]}
          GreenCurve {keys [<at 0.1285 0.432 RCIM_Cubic>,<at 0.3093 1>,<at 0.5661 1>,<at 0.7637 1>]}
          RedCurve {keys [<at 0.1285 0.0082 RCIM_Cubic>,<at 0.3093 0.1622>,<at 0.5661 0.218>,<at 0.7637 0.6857>]}
         }}))
     )
     "Float (Alpha)"=1
    )) {isSetParameters=true}
  ]
  rendererClasses [&NiagaraSpriteRendererProperties]
  renderers [{rendererClass=&NiagaraSpriteRendererProperties}]
 }
 FN_MultiStars {
  emitterUpdateScript [
   &EmitterState()
   &SpawnBurst_Instantaneous(Spawn Count=20, Use Spawn Probability=false)
  ]
  particleSpawnScript [
   &InitializeParticle(
    Color=&UniformRangedLinearColor(
     Minimum=&UniformRangedLinearColor(Minimum=<rgba 1 0 0.0781 1>, Maximum=<rgba 0.9563 1 0 1>)
     Maximum=&UniformRangedLinearColor(Minimum=<rgba 0.2887 1 0 1>, Maximum=<rgba 0 0.1668 1 1>)
    )
    Lifetime=&UniformRangedFloat(Minimum=0.3, Maximum=0.75)
    Module.Write Color=true
    Module.Write Lifetime=true
    Module.Write Mass=true
    Module.Write Material Random=true
    Module.Write Position=false
    Module.Write Scale=false
    Module.Write SpriteRotation=false
    Module.Write SpriteSize=true
    Position=&SimulationPosition()
    Sprite Size=&Vector2DFromFloat(Value=&UniformRangedFloat(Minimum=5, Maximum=12))
   )
   &InheritVelocity(Module.Limit Speed=false, Velocity Scale=<0.25 0.25 0.25>)
   &AddVelocity(Velocity=&UniformRangedVector(Minimum=<100 -500 -500>, Maximum=<1000 500 500>))
  ]
  particleUpdateScript [
   &ParticleState()
   &GravityForce(Gravity=<0 0 -500>)
   &CurlNoiseForce(
    Noise Frequency:NiagaraFloat=&UniformRangedFloat(Minimum=0.1, Maximum=1)
    Noise Strength=&UniformRangedFloat(Minimum=300, Maximum=600)
    Pan Noise=false
    Pan Noise Field=<1 1 0>
   )
   &Drag(Drag=10, Use Linear Drag=true, Use Rotational Drag=false)
   &SolveForcesAndVelocity(Clamp Velocity=false, Limit Acceleration=false)
   &ScaleColor(
    Scale Alpha=&FloatFromCurve(FloatCurve=DataInterface() {propertyValues {Curve {keys [<at 0 0>,<at 0.0967 0.9767>,<at 1.0011 0.002>]}}})
    ScaleA=true
    ScaleRGB=true
   )
   &ScaleSpriteSize(Scale Factor=&Vector2DFromFloat(Value=&FloatFromCurve(FloatCurve=DataInterface() {propertyValues {Curve {keys [<at -0.0046 0>,<at 0.1746 0.9952>,<at 1.0011 0.002>]}}})))
  ]
  rendererClasses [&NiagaraSpriteRendererProperties]
  renderers [{rendererClass=&NiagaraSpriteRendererProperties}]
 }
}
summary {
 systemName=NS_Emote_HandsUp_Burst
 emitters [
  {emitterName=FN_GreenStars, rendererClasses [&NiagaraSpriteRendererProperties]}
  {emitterName=FN_MultiStars, rendererClasses [&NiagaraSpriteRendererProperties]}
 ]
 userVariables [
  User.bIsFrontEndPreview {type=NiagaraBool}
  User.PlayerSpeed {type=NiagaraFloat, defaultValue=0, description=""}
  User.RibbonIDOffset {type=NiagaraInt32}
  User.SkeletalMesh {
   type=NiagaraDataInterfaceSkeletalMesh
   defaultValue {
    struct=NiagaraExt_VariableValue_DataInterface
    value {dataInterface=&self:NiagaraDataInterfaceSkeletalMesh_0}
   }
  }
 ]
}
