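# Niagara system for the spawn moment of the Hands Up emote: a short cyan-blue burst built from four independent sprite emitters.
# A large, brief central glow drives Color and Color2 from a shared gradient, while a slightly delayed burst adds faint blue-violet flares.
# FN_GreenStars scatters shrinking particles with curl noise, drag and gravity, and Starts sends up HDR sparks that flash then fade.
# Skeletal-mesh bone placement in FN_GreenStars is disabled.
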NiagaraSystem /BRCosmetics/Athena/Effects/Particles/Emotes/NS_Emote_HandsUp/NS_Emote_HandsUp_Spawn
dependencies {
 usedDataInterfaces [&NiagaraDataInterfaceCurve, &NiagaraDataInterfaceColorCurve]
 usedDynamicInputs []
 usedModules [/Niagara/Modules/System/SystemLifeCycle, &SystemState]
}
emitters {
 FN_GreenStars {
  emitterSpawnScript [&FN_PreserveSkydiveUserVariables() {module=PreserveUserVariables}]
  emitterUpdateScript [
   &EmitterState()
   &SpawnBurst_Instantaneous(Spawn Count=5, Use Spawn Probability=false)
  ]
  particleSpawnScript [
   /Game/Effects/Niagara/Modules/Groups/FN_SetSpawnGroup_v1/FN_SetSpawnGroup(Emitter Bone Count=2) {enabled=false}
   /Niagara/Modules/Spawn/Initialization/InitializeParticle(
    Color=&UniformRangedLinearColor(Minimum=<rgba 0 0.7923 1 1>, Maximum=<rgba 0.5548 0.6742 0.917 1>)
    Lifetime=&UniformRangedFloat(Minimum=0.25, Maximum=0.5)
    Mesh Scale=&Multiply_Vector(
     A=<1 1 1>
     B=&VectorFromFloat(Value=&UniformRangedFloat(Minimum=0.1, Maximum=0.2))
    )
    Module.Write Color=true
    Module.Write Lifetime=true
    Module.Write Mass=true
    Module.Write Material Random=true
    Module.Write Position=false
    Module.Write Scale=true
    Module.Write SpriteRotation=false
    Module.Write SpriteSize=false
    Position=&SimulationPosition()
   )
   &SampleSkeletalMeshSkeleton(Mesh:NiagaraDataInterfaceSkeletalMesh=$User.SkeletalMesh, Specific Bone Index:NiagaraInt32=$Particles.SpawnGroup) {enabled=false}
   /Niagara/Modules/Spawn/Location/SkeletalMeshSkeletonLocation(Module.Inherit Velocity=false, Module.Use Sampled Bone as Ribbon ID=true) {enabled=false}
   /Niagara/Modules/Spawn/Orientation/V2/InitialMeshOrientation(
    Orientation Vector=&UniformRangedVector(Minimum=<-1 -1 -1>, Maximum=<1 1 1>)
    Use Orientation Vector=true
    Use Rotation Vector=false
   )
   /SetVariables(Particles.RotationOrientation:NiagaraFloat=&UniformRangedFloat(Minimum=-1, Maximum=1)) {isSetParameters=true}
   /SetVariables(Particles.RotationSpeed:Vector3f=&UniformRangedVector(Minimum=<-1 -1 -1>, Maximum=<1 1 1>)) {isSetParameters=true}
   &InheritVelocity(Module.Limit Speed=false, Velocity Scale=<0.25 0.25 0.25>)
   &AddVelocity(Velocity=&UniformRangedVector(Minimum=<-500 -500 -500>, Maximum=<500 500 500>))
  ]
  particleUpdateScript [
   &ParticleState()
   &Drag(Drag=7, Use Linear Drag=true, Use Rotational Drag=false)
   /Niagara/Modules/Update/Forces/CurlNoiseForce(
    Noise Frequency:NiagaraFloat=&UniformRangedFloat(Minimum=0.1, Maximum=1)
    Noise Strength=&UniformRangedFloat(Minimum=300, Maximum=600)
    Pan Noise=false
    Pan Noise Field=<1 1 0>
   )
   &GravityForce(Gravity=<0 0 -500>)
   &SolveForcesAndVelocity(Clamp Velocity=false, Limit Acceleration=false)
   &ScaleSpriteSize(Scale Factor=&Multiply_Vector2(
     A=<1 1>
     B=&Vector2DFromFloat(Value=&FloatFromCurve(FloatCurve=DataInterface() {propertyValues {Curve {keys [<at -0.0046 3>,<at 1.0011 0.002>]}}}))
    ))
   /SetVariables(Particles.MultColor:LinearColor=&MakeLinearColorFromVectorAndFloat(
     "Vector (RGB)"=&Multiply_Vector(
      A=&MakeVectorFromLinearColorRGB(Color=$Particles.Color)
      B=&MakeVectorFromLinearColorRGB(Color=&ColorFromCurve(ColorCurve=DataInterface() {propertyValues {
          AlphaCurve {keys [<at -0.0023 1 RCIM_Cubic>,<at 0.9977 1 RCIM_Cubic>]}
          BlueCurve {keys [<at 0.1285 0.432 RCIM_Cubic>,<at 0.3093 1>,<at 0.5661 0.7784>,<at 0.7637 1>]}
          GreenCurve {keys [<at 0.1285 0.3787 RCIM_Cubic>,<at 0.3093 0.619>,<at 0.5661 1>,<at 0.7637 0.3502>]}
          RedCurve {keys [<at 0.1285 0 RCIM_Cubic>,<at 0.3093 0>,<at 0.5661 0.218>,<at 0.7637 0>]}
         }}))
     )
     "Float (Alpha)"=1
    )) {isSetParameters=true}
  ]
  rendererClasses [&NiagaraSpriteRendererProperties]
  renderers [{rendererClass=&NiagaraSpriteRendererProperties}]
 }
 FN_Sprite001 {
  emitterUpdateScript [
   &EmitterState()
   &SpawnRate(Use Spawn Probability=false) {enabled=false}
   &SpawnBurst_Instantaneous(Spawn Count=1, Use Spawn Probability=false)
  ]
  particleSpawnScript [
   &InitializeParticle(
    Lifetime=0.3
    Module.Write Color=true
    Module.Write Lifetime=true
    Module.Write Mass=true
    Module.Write Material Random=true
    Module.Write Position=false
    Module.Write Scale=false
    Module.Write SpriteRotation=true
    Module.Write SpriteSize=true
    Position=&SimulationPosition()
    Sprite Rotation=&UniformRangedFloat(Minimum=0, Maximum=360)
    Sprite Size=&Vector2DFromFloat(Value=&UniformRangedFloat(Minimum=500, Maximum=500))
   )
   &CameraOffset(Camera Offset Amount=&UniformRangedFloat(Minimum=25, Maximum=25))
   &SystemLocation(MaskSpawnGroup=false, Offset=<0 0 -20>)
  ]
  particleUpdateScript [
   &ParticleState()
   /SetVariables(Particles.GradientColor:Vector3f=&MakeVectorFromLinearColorRGB(Color=&MakeLinearColorFromVectorAndFloat(
      "Vector (RGB)"=&MakeVectorFromLinearColorRGB(Color=&ColorFromCurve(ColorCurve=DataInterface() {propertyValues {
          AlphaCurve {keys [<at -0.0023 1 RCIM_Cubic>,<at 0.9977 1 RCIM_Cubic>]}
          BlueCurve {keys [<at 0.1366 1 RCIM_Cubic>,<at 0.7975 1>,<at 0.9071 1.5678 RCIM_Cubic>]}
          GreenCurve {keys [<at 0.1366 0.6643 RCIM_Cubic>,<at 0.7975 0.6207>,<at 0.9071 0 RCIM_Cubic>]}
          RedCurve {keys [<at 0.1366 0.0405 RCIM_Cubic>,<at 0.7975 0.0629>,<at 0.9071 2 RCIM_Cubic>]}
         }}))
      "Float (Alpha)"=1
     ))) {isSetParameters=true}
   /SetVariables(Particles.Color=&MakeLinearColorFromVectorAndFloat(
     "Vector (RGB)"=&Lerp_Vector(
      A=&Multiply_VectorByFloat(Vector=$Particles.GradientColor, Float=5)
      B=&Multiply_VectorByFloat(Vector=&UniformRangedVector(Minimum=$Particles.GradientColor, Maximum=$Particles.GradientColor), Float=5)
      Alpha=&FloatFromCurve(FloatCurve=DataInterface() {propertyValues {Curve {keys [
           {interpMode=RCIM_Cubic, tangentMode=RCTM_User, time=-0.0056, value=0.0478}
           {arriveTangent=0.0255, interpMode=RCIM_Cubic, leaveTangent=0.0255, tangentMode=RCTM_User, time=0.241, value=0.0789}
           {arriveTangent=0.1228, interpMode=RCIM_Cubic, leaveTangent=0.1228, tangentMode=RCTM_User, time=0.5383, value=0.4094}
           {interpMode=RCIM_Cubic, tangentMode=RCTM_User, time=1.0011, value=0.002}
          ]}}})
     )
     "Float (Alpha)"=&FloatFromCurve(FloatCurve=DataInterface() {propertyValues {Curve {keys [
          {interpMode=RCIM_Cubic, tangentMode=RCTM_User, time=-0.0063, value=0.0184}
          {arriveTangent=1.3376, interpMode=RCIM_Cubic, leaveTangent=1.3376, tangentMode=RCTM_User, time=0.096, value=0.9808}
          {arriveTangent=0.3112, interpMode=RCIM_Cubic, leaveTangent=0.3112, tangentMode=RCTM_User, time=0.5407, value=1.3419}
          {arriveTangent=-3.7119, interpMode=RCIM_Cubic, leaveTangent=-3.7119, tangentMode=RCTM_User, time=0.7242, value=0.3749}
          {interpMode=RCIM_Cubic, tangentMode=RCTM_User, time=1.0011, value=0.002}
         ]}}})
    )) {isSetParameters=true}
   &ScaleSpriteSize(Scale Factor=&Vector2DFromFloat(Value=&FloatFromCurve(FloatCurve=DataInterface() {propertyValues {Curve {keys [
          {arriveTangent=2.3096, interpMode=RCIM_Cubic, leaveTangent=2.3096, tangentMode=RCTM_User, time=0, value=0}
          {arriveTangent=-0.0952, interpMode=RCIM_Cubic, leaveTangent=-0.0952, tangentMode=RCTM_User, time=0.431, value=0.3973}
          <at 0.9969 0.0123 RCIM_Cubic>
         ]}}})))
   &SolveForcesAndVelocity(Clamp Velocity=false, Limit Acceleration=false)
   /SetVariables(Particles.Color2=&MakeLinearColorFromVectorAndFloat(
     "Vector (RGB)"=&Lerp_Vector(
      A=&Multiply_VectorByFloat(Vector=$Particles.GradientColor, Float=2)
      B=&Multiply_VectorByFloat(Vector=&UniformRangedVector(Minimum=$Particles.GradientColor, Maximum=$Particles.GradientColor), Float=5)
      Alpha=&FloatFromCurve(FloatCurve=DataInterface() {propertyValues {Curve {keys [
           {interpMode=RCIM_Cubic, tangentMode=RCTM_User, time=-0.0008, value=0.5048}
           {arriveTangent=0.0255, interpMode=RCIM_Cubic, leaveTangent=0.0255, tangentMode=RCTM_User, time=0.0802, value=0.5788}
           {arriveTangent=0.1228, interpMode=RCIM_Cubic, leaveTangent=0.1228, tangentMode=RCTM_User, time=0.5421, value=2.0946}
           {interpMode=RCIM_Cubic, tangentMode=RCTM_User, time=1.0011, value=0.002}
          ]}}})
     )
     "Float (Alpha)"=&FloatFromCurve(
      FloatCurve=DataInterface() {propertyValues {Curve {keys [
          {interpMode=RCIM_Cubic, tangentMode=RCTM_User, time=-0.0063, value=0.0184}
          {arriveTangent=0.3112, interpMode=RCIM_Cubic, leaveTangent=0.3112, tangentMode=RCTM_User, time=0.3734, value=0.0777}
          {arriveTangent=-1.0034, interpMode=RCIM_Cubic, leaveTangent=-1.0034, tangentMode=RCTM_User, time=0.6422, value=0.1616}
          {interpMode=RCIM_Cubic, tangentMode=RCTM_User, time=1.0011, value=0.002}
         ]}}}
      Scale Curve=0.5
     )
    )) {isSetParameters=true}
   &CameraOffset(Camera Offset Amount=25) {module=CameraOffset001}
  ]
  rendererClasses [&NiagaraSpriteRendererProperties]
  renderers [{rendererClass=&NiagaraSpriteRendererProperties}]
 }
 FlareBurst {
  emitterUpdateScript [
   &EmitterScalability(
    Module.bApplySpawnCountScale=false
    Module.bApplySpawnCountScaleByDistanceFraction=false
    Module.bUseMaxDistance=false
    Module.bUseMinDistance=false
   ) {enabled=false}
   &EmitterLifeCycle(MaxLoopCount=1, NextLoopDuration=0.1) {enabled=false}
   &EmitterState()
   &SpawnBurst_Instantaneous(Spawn Count=8, Spawn Time=0.1, Use Spawn Probability=false)
   &SpawnRate(SpawnRate=15, Use Spawn Probability=false) {enabled=false}
  ]
  particleSpawnScript [
   &InitializeParticle(
    Color=&UniformRangedLinearColor(Minimum=<rgba 0.0668 0.4843 1 0.05>, Maximum=<rgba 0.3472 0.1667 1 0.04>)
    Lifetime=&UniformRangedFloat(Minimum=0.5, Maximum=0.2)
    Module.Write Color=true
    Module.Write Lifetime=true
    Module.Write Mass=true
    Module.Write Material Random=true
    Module.Write Position=true
    Module.Write Scale=false
    Module.Write SpriteRotation=true
    Module.Write SpriteSize=true
    Position=&SimulationPosition()
    Sprite Rotation=&UniformAOrBFloat(A=90, B=-90)
    Sprite Size=&Vector2DFromFloat(Value=&Multiply_Float(A=1.8, B=&UniformRangedFloat(Minimum=15, Maximum=40)))
   )
   &CameraOffset(Camera Offset Amount=10)
   /Niagara/Modules/Spawn/Location/SphereLocation(MaskSpawnGroup=false, Module.Override Randomness=false, Offset=<0 0 10>, Sphere Radius=20, Surface Only=false)
   &AddVelocity(Velocity=&UniformRangedVector(Minimum=<30 30 30>, Maximum=<-30 -30 -30>))
   &SystemLocation(MaskSpawnGroup=false, Offset=<0 0 -20>)
  ]
  particleUpdateScript [
   &ParticleState()
   &SolveForcesAndVelocity(Clamp Velocity=false, Limit Acceleration=false)
   &ScaleColor(
    Scale Alpha=&FloatFromCurve(FloatCurve=DataInterface() {propertyValues {Curve {keys [
         <at 0 0 RCIM_Cubic>
         {arriveTangent=-0.0426, interpMode=RCIM_Cubic, leaveTangent=-0.0426, tangentMode=RCTM_User, time=0.3771, value=1.0137}
         {arriveTangent=0.0121, interpMode=RCIM_Cubic, leaveTangent=0.0121, time=0.5746, value=0.6676}
         {arriveTangent=-1.5692, interpMode=RCIM_Cubic, leaveTangent=-1.5692, time=0.8024, value=1.0189}
         <at 1 0>
        ]}}})
    ScaleA=true
    ScaleRGB=true
   )
   &ScaleSpriteSize(Scale Factor=&Vector2DFromFloat(Value=&FloatFromCurve(FloatCurve=DataInterface() {propertyValues {Curve {keys [<at 0.0121 0.5565>,<at 0.491 0.8934>]}}})))
  ]
  rendererClasses [&NiagaraSpriteRendererProperties]
  renderers [{rendererClass=&NiagaraSpriteRendererProperties}]
 }
 Starts {
  emitterUpdateScript [
   &EmitterScalability(
    Module.bApplySpawnCountScale=false
    Module.bApplySpawnCountScaleByDistanceFraction=false
    Module.bUseMaxDistance=false
    Module.bUseMinDistance=false
   ) {enabled=false}
   &EmitterLifeCycle(MaxLoopCount=1, NextLoopDuration=0.1) {enabled=false}
   &EmitterState()
   &SpawnBurst_Instantaneous(Spawn Count=8, Use Spawn Probability=false)
   &SpawnRate(SpawnRate=3, Use Spawn Probability=false) {enabled=false}
  ]
  particleSpawnScript [
   &InitializeParticle(
    Color=&UniformRangedLinearColor(Minimum=<rgba 4.974 15.7608 20 1>, Maximum=<rgba 4.974 15.7608 20 1>)
    Lifetime=&UniformRangedFloat(Minimum=0.5, Maximum=0.8)
    Module.Write Color=true
    Module.Write Lifetime=true
    Module.Write Mass=true
    Module.Write Material Random=true
    Module.Write Position=true
    Module.Write Scale=false
    Module.Write SpriteRotation=true
    Module.Write SpriteSize=true
    Position=&SimulationPosition()
    Sprite Rotation=&UniformRangedFloat(Minimum=0.02, Maximum=-0.02)
    Sprite Size=&UniformRangedVector2D(Minimum=<3 10>, Maximum=<5 4>)
   )
   &CameraOffset(Camera Offset Amount=5)
   &ApplyInitialForces()
   &AddVelocity(Velocity=&UniformRangedVector(Minimum=<0 0 250>, Maximum=<1 1 20>))
   &SphereLocation(MaskSpawnGroup=false, Module.Override Randomness=false, Offset=<0 0 -30>, Sphere Radius=25, Surface Only=false)
  ]
  particleUpdateScript [
   &ParticleState()
   &Drag(Drag=10, Use Linear Drag=true, Use Rotational Drag=false)
   &SolveForcesAndVelocity(Clamp Velocity=false, Limit Acceleration=false)
   &ScaleColor(
    Scale Alpha=&FloatFromCurve(FloatCurve=DataInterface() {propertyValues {Curve {keys [{arriveTangent=0.0145, interpMode=RCIM_Cubic, leaveTangent=0.0145, tangentMode=RCTM_User, time=0.2, value=1}, <at 1 0>]}}})
    Scale RGB=&VectorFromFloat(Value=&FloatFromCurve(FloatCurve=DataInterface() {propertyValues {Curve {keys [<at 0 50>,<at 0.1 1>]}}}))
    ScaleA=true
    ScaleRGB=true
   )
   &ScaleSpriteSize(Scale Factor=&Vector2DFromFloat(Value=&FloatFromCurve(
      FloatCurve=DataInterface() {propertyValues {Curve {keys [<at 0.0121 0.5565>,<at 0.491 0.8934>,<at 1.0011 0.002>]}}}
      Scale Curve=&UniformRangedFloat(Minimum=0.7, Maximum=1.5)
     )))
  ]
  rendererClasses [&NiagaraSpriteRendererProperties]
  renderers [{rendererClass=&NiagaraSpriteRendererProperties}]
 }
}
summary {
 systemName=NS_Emote_HandsUp_Spawn
 emitters [
  {emitterName=Starts, rendererClasses [&NiagaraSpriteRendererProperties]}
  {emitterName=FlareBurst, rendererClasses [&NiagaraSpriteRendererProperties]}
  {emitterName=FN_Sprite001, rendererClasses [&NiagaraSpriteRendererProperties]}
  {emitterName=FN_GreenStars, rendererClasses [&NiagaraSpriteRendererProperties]}
 ]
 userVariables [
  User.bIsFrontEndPreview {type=NiagaraBool}
  User.PlayerSpeed {type=NiagaraFloat, defaultValue=0, description=""}
  User.RibbonIDOffset {type=NiagaraInt32}
  User.SkeletalMesh {
   type=NiagaraDataInterfaceSkeletalMesh
   defaultValue {
    struct=NiagaraExt_VariableValue_DataInterface
    value {dataInterface=&self:NiagaraDataInterfaceSkeletalMesh_0}
   }
  }
 ]
}
